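; Gameplay ability for the ToyFox boss spawn-in, extending a parent ability with knockback and completion handling.
; The knockback event applies the unburrow knockback gameplay effect to the owner, so the boss briefly knocks back everything around it.
; The notify-begin handler calls the parent, then applies a spawn-in-complete effect when the Complete anim notify tag arrives.
; That effect marks the timing for follow-ups such as showing the boss health bar.
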(event EventTryKnockbackWithGameplayEffect
  ;[0 0.15 0.08092 0.5] Apply unburrow knockback ability to the boss.
  ; For a short duration, the boss knocks back everything around it.
  (Ability|ApplyGameplayEffectToOwner "/TidalCrane_ToyFox_Boss/Abilities/UnburrowKnockback/GE_TidalCrane_ToyFox_Boss_UnburrowKnockbackAbility.GE_TidalCrane_ToyFox_Boss_UnburrowKnockbackAbility_C"))

(event EventNotifyBeginTag
  (params (NotifyTag :type Gameplay%20Tag%20Structure))
  ;[0.07454 0.01642 0.15 0.5] NotifyBeginTag: Apply spawn in complete gameplay effect at the right time. This could be used as the right time to show the bug health bar, etc.
  (|CallParentFunction NotifyTag)
  (if (== NotifyTag "(TagName=\"TidalCrane.ToyFox.AnimNotify.Ability.SpawnIn.Complete\")")
    (Ability|ApplyGameplayEffectToOwner "/TidalCrane_ToyFox_Boss/Abilities/SpawnIn/GE_TidalCrane_ToyFox_Boss_SpawnInComplete.GE_TidalCrane_ToyFox_Boss_SpawnInComplete_C")))
